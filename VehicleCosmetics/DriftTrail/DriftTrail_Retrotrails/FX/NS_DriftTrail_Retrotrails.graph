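# Niagara system for the Retrotrails vehicle drift trail cosmetic, drawing glowing trails and wheel wraps on two wheels.
# Three ribbon emitters, one wide and two narrower side ones, stream backward from offset skeletal positions, briefly inheriting vehicle movement.
# Age curves drive two color layers, widths and a sine/cosine "Yellow WPO" spiral position across three renderers each.
# A mesh emitter pins wheel-sized wraps to bones; FN_DisableTireFX masks all emitters per wheel; painted-variant overrides are disabled.
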NiagaraSystem /VehicleCosmetics/DriftTrail/DriftTrail_Retrotrails/FX/NS_DriftTrail_Retrotrails
dependencies {
 usedDataInterfaces [
  &NiagaraDataInterfaceVectorCurve
  &NiagaraDataInterfaceCurve
  &NiagaraDataInterfaceColorCurve
  &NiagaraDataInterfaceSkeletalMesh
 ]
 usedDynamicInputs []
 usedModules [
  &self:SystemSpawnScript.SystemScriptSource.SystemScriptGraph.NiagaraNodeAssignment_0.SetVariables_#1
  &FN_UnpackVehicleFlags
  /VehicleCosmetics/FX/Niagara/Modules/VC_SetVehicleSpeedAttribs
  &SystemState
 ]
}
emitters {
 NE_VC_DriftTrail_Ribbon_Parent001 {
  emitterUpdateScript [
   &EmitterState()
   &VC_RateQualityHandler(Epic Spawn Rate=30, High Spawn Rate=30, Low Spawn Rate=30, Use Custom Rate?=true)
   &VC_SpawnRibbon_v1(Number of Ribbons=2, SpawnRate=$Emitter.Spawn Rate)
  ]
  particleSpawnScript [
   &VC_PaintedVariantsOverride() {enabled=false}
   &VC_InitializeRibbon_v1(
    Custom Offset=true
    Enable Skeletal Position=true
    Lifetime=1
    Local Vector Offset=&AddVectorToPosition(Position=<-38 0 -25>, Vector=$User.offset)
    Offset Amount=1
    Ribbon Width=160
   )
   &AddVelocity(Velocity=<-1200 0 0>) {module=AddVelocity001}
  ]
  particleUpdateScript [
   &FN_DisableTireFX(Mask Index Attribute:NiagaraInt32=$Particles.SpawnGroup, Mask Index FL=2, Mask Index FR=3, Mask Index RL=0, Mask Index RR=1)
   &ParticleState()
   &InheritSourceMovement(
    Applied Position Scale=&VectorFromCurve(VectorCurve=DataInterface() {propertyValues {
       XCurve {keys [<at 0 0.8 RCIM_Cubic>, {interpMode=RCIM_Cubic, leaveTangent=-1.655, time=0.5, value=0}]}
       YCurve {keys [<at 0 0.8 RCIM_Cubic>, {interpMode=RCIM_Cubic, leaveTangent=-1.655, time=0.5, value=0}]}
       ZCurve {keys [<at 0 0.8 RCIM_Cubic>, {interpMode=RCIM_Cubic, leaveTangent=-1.655, time=0.5, value=0}]}
      }})
    ApplyForce=false
    ApplyPosition=true
    ApplyPositionFalloff=false
   )
   &SolveForcesAndVelocity(Clamp Velocity=false, Limit Acceleration=false)
   &VC_UpdateRibbon_v1(Enable Skeletal Position=true)
   &ScaleRibbonWidth(Ribbon Width Scale=&FloatFromCurve(
     FloatCurve=DataInterface() {propertyValues {Curve {
        defaultValue=1
        keys [
         <at 0 0.2>
         {arriveTangent=-0.8849, interpMode=RCIM_Cubic, leaveTangent=-0.8849, tangentMode=RCTM_User, time=0.2363, value=0.5}
         {interpMode=RCIM_Cubic, leaveTangent=1.2695, time=1, value=0.1}
        ]
       }}}
     Scale Curve=0.5
    ))
   /SetVariables(
    Particles.R1:LinearColor=&Multiply_LinearColorByFloat(
     Linear Color=&MultiplyLinearColor(
      Linear Color=&ColorFromCurve(ColorCurve=DataInterface() {propertyValues {
         AlphaCurve {keys [<at 0.2646 1>,<at 0.6034 1>]}
         BlueCurve {keys [<at 0 0.1615 RCIM_Cubic>,<at 0.1232 0.224>,<at 0.2565 1>,<at 0.3591 1>,<at 0.5237 0.15 RCIM_Cubic>]}
         GreenCurve {keys [<at 0 0.0103 RCIM_Cubic>,<at 0.1232 0.0096>,<at 0.2565 0>,<at 0.3591 0.049>,<at 0.5237 0.279 RCIM_Cubic>]}
         RedCurve {keys [<at 0 0 RCIM_Cubic>,<at 0.1232 0.0062>,<at 0.2565 0.0852>,<at 0.3591 0.2857>,<at 0.5237 1 RCIM_Cubic>]}
        }})
      Scale Factor=&MakeVector4FromVectorAndFloat(
       "Vector (RGB)"=&VectorFromFloat(Value=&FloatFromCurve(
         FloatCurve=DataInterface() {propertyValues {Curve {
            defaultValue=0
            keys [
             {arriveTangent=0.0397, interpMode=RCIM_Cubic, leaveTangent=0.0397, tangentMode=RCTM_User, time=0, value=0.3}
             {arriveTangent=0.0912, interpMode=RCIM_Cubic, leaveTangent=0.0912, tangentMode=RCTM_User, time=0.005, value=0.3}
             {arriveTangent=-0.0576, interpMode=RCIM_Cubic, leaveTangent=-0.0576, tangentMode=RCTM_User, time=0.5, value=0.012}
            ]
           }}}
         Scale Curve=95
        ))
       "Float (Alpha)"=&FloatFromCurve(
        FloatCurve=DataInterface() {propertyValues {Curve {
           defaultValue=0.4
           keys [
            {arriveTangent=0.0014, interpMode=RCIM_Cubic, tangentMode=RCTM_User, time=0, value=0.1}
            {arriveTangent=-0.0131, interpMode=RCIM_Cubic, leaveTangent=-0.0131, tangentMode=RCTM_User, time=0.1475, value=0.5}
            {arriveTangent=-0.4481, interpMode=RCIM_Cubic, leaveTangent=-0.4481, tangentMode=RCTM_User, time=0.5, value=0.08}
            {interpMode=RCIM_Cubic, leaveTangent=-1.1916, time=1, value=0}
           ]
          }}}
        Scale Curve=1.2
       )
      )
     )
     Float=1
    )
    Particles.R2:LinearColor=&Multiply_LinearColorByFloat(
     Linear Color=&MultiplyLinearColor(
      Linear Color=&ColorFromCurve(ColorCurve=DataInterface() {propertyValues {
         AlphaCurve {keys [<at 0.2646 1>,<at 0.6034 1>]}
         BlueCurve {keys [<at 0.3386 0.1615 RCIM_Cubic>,<at 0.5321 0.25>]}
         GreenCurve {keys [<at 0.3386 0.0103 RCIM_Cubic>,<at 0.5321 0.5218>]}
         RedCurve {keys [<at 0.3386 0 RCIM_Cubic>,<at 0.5321 1>]}
        }})
      Scale Factor=&MakeVector4FromVectorAndFloat(
       "Vector (RGB)"=&VectorFromFloat(Value=&FloatFromCurve(
         FloatCurve=DataInterface() {propertyValues {Curve {
            defaultValue=0
            keys [
             {arriveTangent=0.0397, interpMode=RCIM_Cubic, leaveTangent=0.0397, tangentMode=RCTM_User, time=0, value=0.3}
             {arriveTangent=-0.2462, interpMode=RCIM_Cubic, leaveTangent=-0.2462, tangentMode=RCTM_User, time=0.005, value=0.3}
             {arriveTangent=-0.0576, interpMode=RCIM_Cubic, leaveTangent=-0.0576, tangentMode=RCTM_User, time=0.48, value=0.025}
            ]
           }}}
         Scale Curve=85
        ))
       "Float (Alpha)"=&FloatFromCurve(
        FloatCurve=DataInterface() {propertyValues {Curve {
           defaultValue=0.4
           keys [
            {arriveTangent=0.0014, interpMode=RCIM_Cubic, tangentMode=RCTM_User, time=0, value=0.1}
            {arriveTangent=-0.0314, interpMode=RCIM_Cubic, leaveTangent=-0.0314, tangentMode=RCTM_User, time=0.1475, value=0.5}
            {interpMode=RCIM_Cubic, leaveTangent=-1.1916, time=0.993, value=0}
           ]
          }}}
        Scale Curve=1.2
       )
      )
     )
     Float=1
    )
    Particles.WidthR2:NiagaraFloat=&Multiply_Float(
     A=230
     B=&FloatFromCurve(
      FloatCurve=DataInterface() {propertyValues {Curve {
         defaultValue=1
         keys [<at 0 0.25 RCIM_Cubic>, {arriveTangent=-0.25, interpMode=RCIM_Cubic, leaveTangent=-0.25, time=0.5, value=0.6}, <at 1 0 RCIM_Cubic>]
        }}}
      Scale Curve=0.65
     )
    )
   ) {isSetParameters=true, module=SetVariables_#2}
   /SetVariables(Particles.Yellow WPO:NiagaraPosition=&AddVectorToPosition(
     Position=$Particles.Position
     Vector=&TransformVector(
      Vector=&MakeVector(
       X=0
       Y=&Cosine(
        Normalized Angle=&Add_Float(A=&Multiply_Float(A=$Emitter.LoopedAge, B=1), B=$Particles.NormalizedAge)
        Period=1
        Scale=&FloatFromCurve(
         FloatCurve=DataInterface() {propertyValues {Curve {
            defaultValue=1
            keys [
             <at 0 0 RCIM_Cubic>
             {arriveTangent=1.9771, interpMode=RCIM_Cubic, leaveTangent=1.9771, tangentMode=RCTM_User, time=0.2224, value=0.3}
             {interpMode=RCIM_Cubic, tangentMode=RCTM_User, time=1, value=1}
            ]
           }}}
         Scale Curve=2
        )
        Bias=0
       )
       Z=&Sine(
        Normalized Angle=&Add_Float(A=&Multiply_Float(A=$Emitter.LoopedAge, B=1), B=$Particles.NormalizedAge)
        Period=1
        Scale=&FloatFromCurve(
         FloatCurve=DataInterface() {propertyValues {Curve {
            defaultValue=1
            keys [
             <at 0 0 RCIM_Cubic>
             {arriveTangent=1.9771, interpMode=RCIM_Cubic, leaveTangent=1.9771, tangentMode=RCTM_User, time=0.2224, value=0.3}
             {interpMode=RCIM_Cubic, tangentMode=RCTM_User, time=1, value=1}
            ]
           }}}
         Scale Curve=2
        )
        Bias=0
       )
      )
      Destination Space=::Simulation
      Apply Scale=true
     )
    )) {isSetParameters=true, module=SetVariables_#3}
  ]
  rendererClasses [&NiagaraRibbonRendererProperties]
  renderers [
   {rendererClass=&NiagaraRibbonRendererProperties}
   {rendererClass=&NiagaraRibbonRendererProperties, rendererIndex=1}
   {rendererClass=&NiagaraRibbonRendererProperties, rendererIndex=2}
  ]
 }
 NE_VC_DriftTrail_Ribbon_Parent005 {
  emitterUpdateScript [
   &EmitterState()
   &VC_RateQualityHandler(Epic Spawn Rate=30, High Spawn Rate=30, Low Spawn Rate=30, Use Custom Rate?=true)
   &VC_SpawnRibbon_v1(Number of Ribbons=2, SpawnRate=$Emitter.Spawn Rate)
  ]
  particleSpawnScript [
   &VC_PaintedVariantsOverride() {enabled=false}
   &VC_InitializeRibbon_v1(
    Custom Offset=true
    Enable Skeletal Position=true
    Lifetime=1
    Local Vector Offset=&AddVectorToPosition(Position=<-38 8 -25>, Vector=$User.offset)
    Offset Amount=1
    Ribbon Width=55
   )
   &AddVelocity(Velocity=<-1200 0 0>) {module=AddVelocity001}
  ]
  particleUpdateScript [
   &FN_DisableTireFX(Mask Index Attribute:NiagaraInt32=$Particles.SpawnGroup, Mask Index FL=2, Mask Index FR=3, Mask Index RL=0, Mask Index RR=1)
   &ParticleState()
   &InheritSourceMovement(
    Applied Position Scale=&VectorFromCurve(VectorCurve=DataInterface() {propertyValues {
       XCurve {keys [<at 0 0.8 RCIM_Cubic>, {interpMode=RCIM_Cubic, leaveTangent=-1.655, time=0.5, value=0}]}
       YCurve {keys [<at 0 0.8 RCIM_Cubic>, {interpMode=RCIM_Cubic, leaveTangent=-1.655, time=0.5, value=0}]}
       ZCurve {keys [<at 0 0.8 RCIM_Cubic>, {interpMode=RCIM_Cubic, leaveTangent=-1.655, time=0.5, value=0}]}
      }})
    ApplyForce=false
    ApplyPosition=true
    ApplyPositionFalloff=false
   )
   &SolveForcesAndVelocity(Clamp Velocity=false, Limit Acceleration=false)
   &VC_UpdateRibbon_v1(Enable Skeletal Position=true)
   &ScaleRibbonWidth(Ribbon Width Scale=&FloatFromCurve(
     FloatCurve=DataInterface() {propertyValues {Curve {
        defaultValue=1
        keys [
         <at 0 0.2>
         {arriveTangent=-0.8849, interpMode=RCIM_Cubic, leaveTangent=-0.8849, tangentMode=RCTM_User, time=0.2363, value=0.5}
         {interpMode=RCIM_Cubic, leaveTangent=1.2695, time=1, value=0.1}
        ]
       }}}
     Scale Curve=0.5
    ))
   /SetVariables(
    Particles.R1:LinearColor=&Multiply_LinearColorByFloat(
     Linear Color=&MultiplyLinearColor(
      Linear Color=&ColorFromCurve(ColorCurve=DataInterface() {propertyValues {
         AlphaCurve {keys [<at 0.2646 1>,<at 0.6034 1>]}
         BlueCurve {keys [<at 0 0.375 RCIM_Cubic>,<at 0.1149 0.4792>,<at 0.2218 1>,<at 0.4345 0.15 RCIM_Cubic>]}
         GreenCurve {keys [<at 0 0.0239 RCIM_Cubic>,<at 0.1149 0.0135>,<at 0.2218 0>,<at 0.4345 0.3185 RCIM_Cubic>]}
         RedCurve {keys [<at 0 0 RCIM_Cubic>,<at 0.1149 0.0228>,<at 0.2218 0.0852>,<at 0.4345 1 RCIM_Cubic>]}
        }})
      Scale Factor=&MakeVector4FromVectorAndFloat(
       "Vector (RGB)"=&VectorFromFloat(Value=&FloatFromCurve(
         FloatCurve=DataInterface() {propertyValues {Curve {
            defaultValue=0
            keys [
             {arriveTangent=-0.0258, interpMode=RCIM_Cubic, leaveTangent=-0.0258, tangentMode=RCTM_User, time=0, value=0.3}
             {arriveTangent=-0.2462, interpMode=RCIM_Cubic, leaveTangent=-0.2462, tangentMode=RCTM_User, time=0.025, value=0.3}
             {arriveTangent=-0.0331, interpMode=RCIM_Cubic, leaveTangent=-0.0331, tangentMode=RCTM_User, time=0.5, value=0.05}
             <at 1 0.05 RCIM_Cubic>
            ]
           }}}
         Scale Curve=65
        ))
       "Float (Alpha)"=&FloatFromCurve(
        FloatCurve=DataInterface() {propertyValues {Curve {
           defaultValue=0.4
           keys [
            {arriveTangent=0.0014, interpMode=RCIM_Cubic, tangentMode=RCTM_User, time=0, value=0.1}
            {arriveTangent=-0.0131, interpMode=RCIM_Cubic, leaveTangent=-0.0131, tangentMode=RCTM_User, time=0.1475, value=0.5}
            {arriveTangent=-0.2314, interpMode=RCIM_Cubic, leaveTangent=-0.2314, tangentMode=RCTM_User, time=0.5, value=0.08}
            {arriveTangent=0.0271, interpMode=RCIM_Cubic, leaveTangent=0.0271, tangentMode=RCTM_User, time=1, value=0}
           ]
          }}}
        Scale Curve=1.2
       )
      )
     )
     Float=1
    )
    Particles.R2:LinearColor=&Multiply_LinearColorByFloat(
     Linear Color=&MultiplyLinearColor(
      Linear Color=&ColorFromCurve(ColorCurve=DataInterface() {propertyValues {
         AlphaCurve {keys [<at 0.2646 1>,<at 0.6034 1>]}
         BlueCurve {keys [<at 0.3386 0.1615 RCIM_Cubic>,<at 0.5321 0.25>]}
         GreenCurve {keys [<at 0.3386 0.0103 RCIM_Cubic>,<at 0.5321 0.5218>]}
         RedCurve {keys [<at 0.3386 0 RCIM_Cubic>,<at 0.5321 1>]}
        }})
      Scale Factor=&MakeVector4FromVectorAndFloat(
       "Vector (RGB)"=&VectorFromFloat(Value=&FloatFromCurve(
         FloatCurve=DataInterface() {propertyValues {Curve {
            defaultValue=0
            keys [
             {arriveTangent=0.0397, interpMode=RCIM_Cubic, leaveTangent=0.0397, tangentMode=RCTM_User, time=0, value=0.3}
             {arriveTangent=-0.2462, interpMode=RCIM_Cubic, leaveTangent=-0.2462, tangentMode=RCTM_User, time=0.005, value=0.3}
             {arriveTangent=-0.0576, interpMode=RCIM_Cubic, leaveTangent=-0.0576, tangentMode=RCTM_User, time=0.48, value=0.025}
            ]
           }}}
         Scale Curve=85
        ))
       "Float (Alpha)"=&FloatFromCurve(
        FloatCurve=DataInterface() {propertyValues {Curve {
           defaultValue=0.4
           keys [
            {arriveTangent=0.0014, interpMode=RCIM_Cubic, tangentMode=RCTM_User, time=0, value=0.1}
            {arriveTangent=-0.0131, interpMode=RCIM_Cubic, leaveTangent=-0.0131, tangentMode=RCTM_User, time=0.1475, value=0.5}
            {arriveTangent=-0.2164, interpMode=RCIM_Cubic, leaveTangent=-0.2164, tangentMode=RCTM_User, time=0.5, value=0.08}
            {interpMode=RCIM_Cubic, leaveTangent=-1.1916, time=1, value=0}
           ]
          }}}
        Scale Curve=1.2
       )
      )
     )
     Float=1
    )
    Particles.WidthR2:NiagaraFloat=&Multiply_Float(
     A=90
     B=&FloatFromCurve(
      FloatCurve=DataInterface() {propertyValues {Curve {
         defaultValue=1
         keys [
          <at 0 0.6 RCIM_Cubic>
          {arriveTangent=0.0136, interpMode=RCIM_Cubic, leaveTangent=0.0136, tangentMode=RCTM_User, time=0.5, value=0.6}
          {interpMode=RCIM_Cubic, leaveTangent=-1.0208, time=1, value=0}
         ]
        }}}
      Scale Curve=0.65
     )
    )
   ) {isSetParameters=true, module=SetVariables_#2}
   /SetVariables(Particles.Yellow WPO:NiagaraPosition=&AddVectorToPosition(
     Position=$Particles.Position
     Vector=&TransformVector(
      Vector=&MakeVector(
       X=0
       Y=&Cosine(
        Normalized Angle=&Add_Float(A=&Multiply_Float(A=$Emitter.LoopedAge, B=-2), B=$Particles.NormalizedAge)
        Period=1
        Scale=&FloatFromCurve(
         FloatCurve=DataInterface() {propertyValues {Curve {
            defaultValue=1
            keys [
             <at 0 0 RCIM_Cubic>
             {arriveTangent=1.9104, interpMode=RCIM_Cubic, leaveTangent=1.9104, tangentMode=RCTM_User, time=0.2, value=1.5}
             {interpMode=RCIM_Cubic, tangentMode=RCTM_User, time=1, value=2}
            ]
           }}}
         Scale Curve=-8
        )
        Bias=0
       )
       Z=&Sine(
        Normalized Angle=&Add_Float(A=&Multiply_Float(A=$Emitter.LoopedAge, B=-2), B=$Particles.NormalizedAge)
        Period=1
        Scale=&FloatFromCurve(
         FloatCurve=DataInterface() {propertyValues {Curve {
            defaultValue=1
            keys [
             <at 0 0 RCIM_Cubic>
             {arriveTangent=1.9771, interpMode=RCIM_Cubic, leaveTangent=1.9771, tangentMode=RCTM_User, time=0.2224, value=0.3}
             {interpMode=RCIM_Cubic, tangentMode=RCTM_User, time=1, value=1.5}
            ]
           }}}
         Scale Curve=-12
        )
        Bias=0
       )
      )
      Destination Space=::Simulation
      Apply Scale=true
     )
    )) {isSetParameters=true, module=SetVariables_#3}
  ]
  rendererClasses [&NiagaraRibbonRendererProperties]
  renderers [
   {rendererClass=&NiagaraRibbonRendererProperties}
   {rendererClass=&NiagaraRibbonRendererProperties, rendererIndex=1}
   {rendererClass=&NiagaraRibbonRendererProperties, rendererIndex=2}
  ]
 }
 NE_VC_DriftTrail_Ribbon_Parent007 {
  emitterUpdateScript [
   &EmitterState()
   &VC_RateQualityHandler(Epic Spawn Rate=30, High Spawn Rate=30, Low Spawn Rate=30, Use Custom Rate?=true)
   &VC_SpawnRibbon_v1(Number of Ribbons=2, SpawnRate=$Emitter.Spawn Rate)
  ]
  particleSpawnScript [
   &VC_PaintedVariantsOverride() {enabled=false}
   &VC_InitializeRibbon_v1(
    Custom Offset=true
    Enable Skeletal Position=true
    Lifetime=1
    Local Vector Offset=&AddVectorToPosition(Position=<-38 -8 -25>, Vector=$User.offset)
    Offset Amount=1
    Ribbon Width=55
   )
   &AddVelocity(Velocity=<-1200 0 0>) {module=AddVelocity001}
  ]
  particleUpdateScript [
   &FN_DisableTireFX(Mask Index Attribute:NiagaraInt32=$Particles.SpawnGroup, Mask Index FL=2, Mask Index FR=3, Mask Index RL=0, Mask Index RR=1)
   &ParticleState()
   &InheritSourceMovement(
    Applied Position Scale=&VectorFromCurve(VectorCurve=DataInterface() {propertyValues {
       XCurve {keys [<at 0 0.8 RCIM_Cubic>, {interpMode=RCIM_Cubic, leaveTangent=-1.655, time=0.5, value=0}]}
       YCurve {keys [<at 0 0.8 RCIM_Cubic>, {interpMode=RCIM_Cubic, leaveTangent=-1.655, time=0.5, value=0}]}
       ZCurve {keys [<at 0 0.8 RCIM_Cubic>, {interpMode=RCIM_Cubic, leaveTangent=-1.655, time=0.5, value=0}]}
      }})
    ApplyForce=false
    ApplyPosition=true
    ApplyPositionFalloff=false
   )
   &SolveForcesAndVelocity(Clamp Velocity=false, Limit Acceleration=false)
   &VC_UpdateRibbon_v1(Enable Skeletal Position=true)
   &ScaleRibbonWidth(Ribbon Width Scale=&FloatFromCurve(
     FloatCurve=DataInterface() {propertyValues {Curve {
        defaultValue=1
        keys [
         <at 0 0.2>
         {arriveTangent=-0.8849, interpMode=RCIM_Cubic, leaveTangent=-0.8849, tangentMode=RCTM_User, time=0.2363, value=0.5}
         {interpMode=RCIM_Cubic, leaveTangent=1.2695, time=1, value=0.1}
        ]
       }}}
     Scale Curve=0.5
    ))
   /SetVariables(
    Particles.R1:LinearColor=&Multiply_LinearColorByFloat(
     Linear Color=&MultiplyLinearColor(
      Linear Color=&ColorFromCurve(ColorCurve=DataInterface() {propertyValues {
         AlphaCurve {keys [<at 0.2646 1>,<at 0.6034 1>]}
         BlueCurve {keys [<at 0 0.375 RCIM_Cubic>,<at 0.1149 0.4792>,<at 0.2218 1>,<at 0.4345 0.15 RCIM_Cubic>]}
         GreenCurve {keys [<at 0 0.0239 RCIM_Cubic>,<at 0.1149 0.0135>,<at 0.2218 0>,<at 0.4345 0.3185 RCIM_Cubic>]}
         RedCurve {keys [<at 0 0 RCIM_Cubic>,<at 0.1149 0.0228>,<at 0.2218 0.0852>,<at 0.4345 1 RCIM_Cubic>]}
        }})
      Scale Factor=&MakeVector4FromVectorAndFloat(
       "Vector (RGB)"=&VectorFromFloat(Value=&FloatFromCurve(
         FloatCurve=DataInterface() {propertyValues {Curve {
            defaultValue=0
            keys [
             {arriveTangent=-0.0258, interpMode=RCIM_Cubic, leaveTangent=-0.0258, tangentMode=RCTM_User, time=0, value=0.3}
             {arriveTangent=-0.2462, interpMode=RCIM_Cubic, leaveTangent=-0.2462, tangentMode=RCTM_User, time=0.025, value=0.3}
             {arriveTangent=-0.0331, interpMode=RCIM_Cubic, leaveTangent=-0.0331, tangentMode=RCTM_User, time=0.5, value=0.05}
             <at 1 0.05 RCIM_Cubic>
            ]
           }}}
         Scale Curve=65
        ))
       "Float (Alpha)"=&FloatFromCurve(
        FloatCurve=DataInterface() {propertyValues {Curve {
           defaultValue=0.4
           keys [
            {arriveTangent=0.0014, interpMode=RCIM_Cubic, tangentMode=RCTM_User, time=0, value=0.1}
            {arriveTangent=-0.0131, interpMode=RCIM_Cubic, leaveTangent=-0.0131, tangentMode=RCTM_User, time=0.1475, value=0.5}
            {arriveTangent=-0.2314, interpMode=RCIM_Cubic, leaveTangent=-0.2314, tangentMode=RCTM_User, time=0.5, value=0.08}
            {arriveTangent=0.0271, interpMode=RCIM_Cubic, leaveTangent=0.0271, tangentMode=RCTM_User, time=1, value=0}
           ]
          }}}
        Scale Curve=1.2
       )
      )
     )
     Float=1
    )
    Particles.R2:LinearColor=&Multiply_LinearColorByFloat(
     Linear Color=&MultiplyLinearColor(
      Linear Color=&ColorFromCurve(ColorCurve=DataInterface() {propertyValues {
         AlphaCurve {keys [<at 0.2646 1>,<at 0.6034 1>]}
         BlueCurve {keys [<at 0.3386 0.1615 RCIM_Cubic>,<at 0.5321 0.25>]}
         GreenCurve {keys [<at 0.3386 0.0103 RCIM_Cubic>,<at 0.5321 0.5218>]}
         RedCurve {keys [<at 0.3386 0 RCIM_Cubic>,<at 0.5321 1>]}
        }})
      Scale Factor=&MakeVector4FromVectorAndFloat(
       "Vector (RGB)"=&VectorFromFloat(Value=&FloatFromCurve(
         FloatCurve=DataInterface() {propertyValues {Curve {
            defaultValue=0
            keys [
             {arriveTangent=0.0397, interpMode=RCIM_Cubic, leaveTangent=0.0397, tangentMode=RCTM_User, time=0, value=0.3}
             {arriveTangent=-0.2462, interpMode=RCIM_Cubic, leaveTangent=-0.2462, tangentMode=RCTM_User, time=0.005, value=0.3}
             {arriveTangent=-0.0576, interpMode=RCIM_Cubic, leaveTangent=-0.0576, tangentMode=RCTM_User, time=0.48, value=0.025}
            ]
           }}}
         Scale Curve=85
        ))
       "Float (Alpha)"=&FloatFromCurve(
        FloatCurve=DataInterface() {propertyValues {Curve {
           defaultValue=0.4
           keys [
            {arriveTangent=0.0014, interpMode=RCIM_Cubic, tangentMode=RCTM_User, time=0, value=0.1}
            {arriveTangent=-0.0131, interpMode=RCIM_Cubic, leaveTangent=-0.0131, tangentMode=RCTM_User, time=0.1475, value=0.5}
            {arriveTangent=-0.2164, interpMode=RCIM_Cubic, leaveTangent=-0.2164, tangentMode=RCTM_User, time=0.5, value=0.08}
            {interpMode=RCIM_Cubic, leaveTangent=-1.1916, time=1, value=0}
           ]
          }}}
        Scale Curve=1.2
       )
      )
     )
     Float=1
    )
    Particles.WidthR2:NiagaraFloat=&Multiply_Float(
     A=90
     B=&FloatFromCurve(
      FloatCurve=DataInterface() {propertyValues {Curve {
         defaultValue=1
         keys [
          <at 0 0.6 RCIM_Cubic>
          {arriveTangent=0.0136, interpMode=RCIM_Cubic, leaveTangent=0.0136, tangentMode=RCTM_User, time=0.5, value=0.6}
          {interpMode=RCIM_Cubic, leaveTangent=-1.0208, time=1, value=0}
         ]
        }}}
      Scale Curve=0.65
     )
    )
   ) {isSetParameters=true, module=SetVariables_#2}
   /SetVariables(Particles.Yellow WPO:NiagaraPosition=&AddVectorToPosition(
     Position=$Particles.Position
     Vector=&TransformVector(
      Vector=&MakeVector(
       X=0
       Y=&Cosine(
        Normalized Angle=&Add_Float(A=&Multiply_Float(A=$Emitter.LoopedAge, B=-2), B=$Particles.NormalizedAge)
        Period=1
        Scale=&FloatFromCurve(
         FloatCurve=DataInterface() {propertyValues {Curve {
            defaultValue=1
            keys [
             <at 0 0 RCIM_Cubic>
             {arriveTangent=1.9771, interpMode=RCIM_Cubic, leaveTangent=1.9771, tangentMode=RCTM_User, time=0.2, value=1.5}
             {interpMode=RCIM_Cubic, tangentMode=RCTM_User, time=1, value=2}
            ]
           }}}
         Scale Curve=8
        )
        Bias=0
       )
       Z=&Sine(
        Normalized Angle=&Add_Float(A=&Multiply_Float(A=$Emitter.LoopedAge, B=-2), B=$Particles.NormalizedAge)
        Period=1
        Scale=&FloatFromCurve(
         FloatCurve=DataInterface() {propertyValues {Curve {
            defaultValue=1
            keys [
             <at 0 0 RCIM_Cubic>
             {arriveTangent=1.9771, interpMode=RCIM_Cubic, leaveTangent=1.9771, tangentMode=RCTM_User, time=0.2224, value=0.3}
             {interpMode=RCIM_Cubic, tangentMode=RCTM_User, time=1, value=1.5}
            ]
           }}}
         Scale Curve=-12
        )
        Bias=0
       )
      )
      Destination Space=::Simulation
      Apply Scale=true
     )
    )) {isSetParameters=true, module=SetVariables_#3}
  ]
  rendererClasses [&NiagaraRibbonRendererProperties]
  renderers [
   {rendererClass=&NiagaraRibbonRendererProperties}
   {rendererClass=&NiagaraRibbonRendererProperties, rendererIndex=1}
   {rendererClass=&NiagaraRibbonRendererProperties, rendererIndex=2}
  ]
 }
 NE_VC_WheelWrap004 {
  emitterUpdateScript [
   &EmitterState(
    Inactive Response=::NewEnumerator1
    Life Cycle Mode=::NewEnumerator1
    Loop Behavior=::NewEnumerator1
    Loop Duration Mode=::NewEnumerator1
    UseLoopDelay=false
   )
   &SpawnBurst_Instantaneous(Spawn Count=2, Use Spawn Probability=false, UseLoopCountLimit=false)
  ]
  particleSpawnScript [
   &VC_PaintedVariantsOverride(
    Black=<rgba 0.015 0.015 0.015 10>
    BurntSienna=<rgba 0.2396 0.057 0.0311 1>
    Cobalt=<rgba 0 0.1524 1.2 1>
    Crimson=<rgba 0.4635 0 0.0126 5>
    ForestGreen=<rgba 0 0.276 0.0139 5>
    Grey=<rgba 0.1514 0.1514 0.1615 5>
    Lime=<rgba 0.2599 0.4427 0 5>
    Orange=<rgba 0.6724 0.1426 0.0038 10>
    Pink=<rgba 0.7347 0.2348 0.7448 10>
    Purple=<rgba 0.1621 0.0606 0.8958 10>
    Saffron=<rgba 0.8854 0.6788 0 5>
    SkyBlue=<rgba 0.0197 0.3277 0.6 5>
    TitaniumWhite=<rgba 0.5596 0.5596 0.5625 5>
   ) {enabled=false}
   &InitializeParticle(
    Color Mode=::NewEnumerator0
    Lifetime=5
    Mesh Scale Mode=::NewEnumerator4
    Sprite Rotation Mode=::NewEnumerator2
    Sprite Size Mode=::NewEnumerator3
    UsePositionOffset=false
    Write Material Random=true
   )
   &PartitionParticles()
   &SkeletalMeshLocation(
    Bone / Socket Index=&Add_Integer(A=$Particles.PartitionParticles.Partition, B=4)
    Bone / Socket Sampling Mode=::NewEnumerator5
    Interpolate Mesh Sampling=false
    MaskSpawnGroup=false
    Mesh Sampling Type=::NewEnumerator6
    Skeletal Mesh=$System.SkeletalMesh
   )
   &InitialMeshOrientation(Mesh Orientation Mode=::NewEnumerator7, Rotation=<0 0 0.5>, Use Rotation Vector=true)
   &Color(Color=<rgba 0 1 0 1>) {enabled=false}
  ]
  particleUpdateScript [
   &FN_DisableTireFX(
    Mask Index Attribute:NiagaraInt32=$Particles.PartitionParticles.Partition
    Mask Index FL=2
    Mask Index FR=3
    Mask Index RL=0
    Mask Index RR=1
   )
   &VC_WheelWrapScale(Wheel Index Attribute:NiagaraInt32=&Add_Integer(A=$Particles.PartitionParticles.Partition, B=2))
   &ParticleState(Kill Particles When Lifetime Has Elapsed=false, Loop Particles Lifetime=true)
   &UpdateMeshOrientation(Rotation Coordinate Space=::NewEnumerator2, Rotation Vector=<0 1 0>) {enabled=false}
   &SkeletalMeshLocation(
    Bone / Socket Index=&Add_Integer(A=$Particles.PartitionParticles.Partition, B=4)
    Bone / Socket Sampling Mode=::NewEnumerator5
    MaskSpawnGroup=false
    Mesh Sampling Type=::NewEnumerator6
    Skeletal Mesh=$System.SkeletalMesh
   ) {module=SkeletalMeshLocation001}
   &SolveForcesAndVelocity(Clamp Velocity=false, Limit Acceleration=false)
   &ScaleColor(
    Scale RGB=&VectorFromCurve(VectorCurve=DataInterface() {propertyValues {XCurve {keys [<at 0 1>,<at 1 2>]}, YCurve {keys [<at 0 1>,<at 1 2>]}, ZCurve {keys [<at 0 1>,<at 1 2>]}}})
    ScaleA=true
    ScaleRGB=true
   )
   /SetVariables(
    Particles.BlueColor:LinearColor=&Multiply_LinearColorByFloat(
     Linear Color=&MultiplyLinearColor(
      Linear Color=<rgba 1.5 1.5 1.5 15>
      Scale Factor=&MakeVector4FromVectorAndFloat("Vector (RGB)"=&VectorFromFloat(Value=2.5), "Float (Alpha)"=1)
     )
     Float=1
    )
    Particles.Linear Color=&Multiply_LinearColorByFloat(
     Linear Color=&MultiplyLinearColor(
      Linear Color=<rgba 1.5 1.5 1.5 15>
      Scale Factor=&MakeVector4FromVectorAndFloat("Vector (RGB)"=&VectorFromFloat(Value=2), "Float (Alpha)"=1)
     )
     Float=1
    )
   ) {isSetParameters=true}
  ]
  rendererClasses [&NiagaraMeshRendererProperties]
  renderers [
   {rendererClass=&NiagaraMeshRendererProperties, rendererIndex=0}
   {rendererClass=&NiagaraMeshRendererProperties, rendererIndex=1}
   {rendererClass=&NiagaraMeshRendererProperties, rendererIndex=2}
  ]
 }
}
summary {
 systemName=NS_DriftTrail_Retrotrails
 emitters [
  {emitterName=NE_VC_WheelWrap004, rendererClasses [&NiagaraMeshRendererProperties]}
  {emitterName=NE_VC_DriftTrail_Ribbon_Parent001, rendererClasses [&NiagaraRibbonRendererProperties]}
  {emitterName=NE_VC_DriftTrail_Ribbon_Parent005, rendererClasses [&NiagaraRibbonRendererProperties]}
  {emitterName=NE_VC_DriftTrail_Ribbon_Parent007, rendererClasses [&NiagaraRibbonRendererProperties]}
 ]
 userVariables [
  User.bIsFrontendPreview {type=NiagaraBool}
  User.bIsPainted {type=NiagaraBool}
  User.ColorVariantID {type=NiagaraInt32}
  User.offset {type=Vector3f, description="offset from wheels(set through cosmetic template)", defaultValue {struct=Vector3f, value=<0 0 0>}}
  User.PaintedColor {type=LinearColor, description="", defaultValue {struct=LinearColor, value=<rgba 1 1 1 1>}}
  User.PaintedScalar {type=NiagaraFloat, defaultValue=1, description=""}
  User.VehicleFlags {type=NiagaraInt32, defaultValue=806289408}
  User.Wheel_Radius_Array {
   type=NiagaraDataInterfaceArrayFloat
   defaultValue {
    struct=NiagaraExt_VariableValue_DataInterface
    value {
     dataInterface=&self:NiagaraDataInterfaceArrayFloat_1
     dataInterfaceClass=&NiagaraDataInterfaceArrayFloat
    }
   }
  }
  User.Wheel_Width_Array {
   type=NiagaraDataInterfaceArrayFloat
   defaultValue {
    struct=NiagaraExt_VariableValue_DataInterface
    value {
     dataInterface=&self:NiagaraDataInterfaceArrayFloat_0
     dataInterfaceClass=&NiagaraDataInterfaceArrayFloat
    }
   }
  }
  User.WheelTouchingArray {
   type=NiagaraDataInterfaceArrayFloat
   defaultValue {
    struct=NiagaraExt_VariableValue_DataInterface
    value {
     dataInterface=&self:NiagaraDataInterfaceArrayFloat_2
     dataInterfaceClass=&NiagaraDataInterfaceArrayFloat
    }
   }
  }
 ]
}
